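# Sound cue for the Boombox gadget's beat explosion, played as an impact-class sound.
# A Random node picks one of three equally weighted explosion waves and feeds it through a Modulator for variation.
# The cue references the shared Boombox effects attenuation asset, alongside override values for distance low-pass filtering and occlusion.
# A concurrency override keeps simultaneous instances to a small limit.
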SoundCue /FNGameplayCues/Sounds/Fort_Gadgets/Boombox/Boombox_BeatExplo_Cue
root=Modulator_0
nodes [
 {id=Modulator_0, children [Random_0]}
 {id=Random_0, children [WavePlayer_1,WavePlayer_2,WavePlayer_3], properties {weights [1,1,1]}}
 {id=WavePlayer_1, properties {soundWaveAssetPtr=&/Boombox_BeatExplo_04}}
 {id=WavePlayer_2, properties {soundWaveAssetPtr=&/Boombox_BeatExplo_01}}
 {id=WavePlayer_3, properties {soundWaveAssetPtr=&/Boombox_BeatExplo_02}}
]
properties {
 attenuationSettings=/Game/Sounds/Fort_Gadgets/Boombox/Boombox_Effects_Att
 bOverrideConcurrency=true
 maxDistance=6000
 soundClassObject=&Fort_Impacts
 volumeMultiplier=0.6
 attenuationOverrides {
  bAttenuateWithLPF=true
  bEnableOcclusion=true
  falloffDistance=4000
  lPFFrequencyAtMax=1000
  lPFRadiusMax=3000
  lPFRadiusMin=1000
  occlusionInterpolationTime=1
  occlusionLowPassFilterFrequency=1000
  occlusionTraceChannel=ECC_GameTraceChannel11
  occlusionVolumeAttenuation=0.6
  stereoSpread=2000
  attenuationShapeExtents {x=0}
 }
 concurrencyOverrides {maxCount=2}
}
